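# Animation Blueprint for the face accessory part of the male large EctoCat player body.
# It copies the pose and curves from the attached parent mesh, then layers the BuffCat_Male_Facial sequence over it with Layered blend per bone.
# The blend weight is the DisableFaceOverride curve value multiplied by one, so the facial layer strengthens as that curve rises.
# The event graph holds only an empty update event.
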AnimBlueprint /BRCosmetics/Characters/Player/Male/Large/Bodies/M_LRG_EctoCat/Meshes/Parts/M_LRG_EctoCat_FaceAcc_AnimBP
target_skeleton=&/M_LRG_EctoCat_FaceAcc_Skeleton
graphs [
 AnimGraph {nodes [
   n1 = GetCurveValue(CurveName=DisableFaceOverride)
   Output Pose(Result=Layered blend per bone(
     BasePose=Copy Pose From Mesh() {bCopyCurves=true, bUseAttachedParent=true}
     BlendPoses_0=Sequence Player 'BuffCat_Male_Facial'() {sequence=/BRCosmetics/Animation/Game/MainPlayer/Facial/BuffCat/BuffCat_Male_Facial}
     BlendWeights_0=float * float(A=1, B=@n1)
    )) {layerGroup=None}
  ]}
 EventGraph {nodes [Event BlueprintUpdateAnimation()]}
]
variables []
